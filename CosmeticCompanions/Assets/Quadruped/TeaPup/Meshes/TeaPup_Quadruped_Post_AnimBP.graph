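# Post-process animation Blueprint for the TeaPup quadruped cosmetic companion, layering secondary tail motion and pose-asset blending onto the incoming pose.
# Two chained Anim Dynamics nodes simulate tail_03 and tail_04 root-relative, using angular springs, tight angular limits and reduced gravity.
# The result returns to local space and feeds a Pose Blend node that applies TeaPup_Quadruped_PoseAsset from the pose's curves.
# The event graph is empty.
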AnimBlueprint /CosmeticCompanions/Assets/Quadruped/TeaPup/Meshes/TeaPup_Quadruped_Post_AnimBP
target_skeleton=&/TeaPup_Quadruped_Skeleton
graphs [
 AnimGraph {nodes [
   n1 = Anim Dynamics - Bone: tail_03(ComponentPose=Local To Component(LocalPose=Input Pose()), Alpha=1, ExternalForce="0,0,0") {
    angularDampingOverride=1
    angularSpringConstant=22000
    bAngularSpring=true
    bOverrideAngularDamping=true
    bUseGravityOverride=true
    gravityScale=0.35
    simulationSpace=RootRelative
    alphaBoolBlend {blendOption=Linear}
    alphaScaleBiasClamp {bias=1, scale=-1}
    boundBone {boneName=tail_03}
    componentAppliedLinearAccClamp=<4000 4000 4000>
    componentLinearAccScale=<1 1 1>
    componentLinearVelScale=<1 1 1>
    gravityOverride {z=-980}
    physicsBodyDefinitions [{
      boundBone {boneName=tail_03}
      boxExtents=<24 24 24>
      constraintSetup {angularLimitsMax=<0 3 10>, angularLimitsMin=<0 -15 -10>, angularTarget=<1 0 0>}
      localJointOffset=<2 0 0>
     }]
   }
   n2 = Anim Dynamics - Bone: tail_04(ComponentPose=@n1, Alpha=1, ExternalForce="0,0,0") {
    angularDampingOverride=1
    angularSpringConstant=22000
    bAngularSpring=true
    bOverrideAngularDamping=true
    bUseGravityOverride=true
    gravityScale=0.35
    simulationSpace=RootRelative
    alphaBoolBlend {blendOption=Linear}
    alphaScaleBiasClamp {bias=1, scale=-1}
    boundBone {boneName=tail_04}
    componentAppliedLinearAccClamp=<4000 4000 4000>
    componentLinearAccScale=<1 1 1>
    componentLinearVelScale=<1 1 1>
    gravityOverride {z=-980}
    physicsBodyDefinitions [{
      boundBone {boneName=tail_04}
      boxExtents=<24 24 24>
      constraintSetup {angularLimitsMax=<0 3 5>, angularLimitsMin=<0 -15 -5>, angularTarget=<1 0 0>}
      localJointOffset=<2 0 0>
     }]
   }
   Output Pose(Result=Pose!(
     PoseAsset:Pose Asset Object Reference=Get PoseAsset1() -> (PoseAsset1:Pose Asset Object Reference)
     SourcePose:Pose Link Structure=Component To Local(ComponentPose=@n2)
    ) -> (Pose:Pose Link Structure) {is=title AnimGraphNode_PoseBlendNode, blendOption=Linear})
  ]}
 EventGraph {links [], nodes []}
]
variables [PoseAsset1 {type=PoseAsset, default=&/TeaPup_Quadruped_PoseAsset}]
